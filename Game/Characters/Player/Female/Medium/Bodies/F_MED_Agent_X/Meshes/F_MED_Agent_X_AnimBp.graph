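# Post-process-style AnimBlueprint for the female medium Agent X body that corrects upper-arm pose during a front-end menu idle.
# It copies the attached parent mesh's pose, then applies small additive bone-space rotations to both upper arms, weighted by FrontEndAlpha.
# Initialization caches the owning FortPlayerPawn's mesh AnimInstance.
# Each update snaps the alpha to 1 while that instance plays the Female_Commando_Idle_03 menu montage, otherwise 0.
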AnimBlueprint /Game/Characters/Player/Female/Medium/Bodies/F_MED_Agent_X/Meshes/F_MED_Agent_X_AnimBp
target_skeleton=&/F_MED_Agent_X_Skeleton
graphs [
 AnimGraph {nodes [
   Input Pose()
   Local To Component()
   n1 = Get FrontEndAlpha() -> (FrontEndAlpha:double)
   n2 = "Transform (Modify) Bone - Bone: upperarm_l"(
    ComponentPose=Local To Component(LocalPose=Copy Pose From Mesh() {bUseAttachedParent=true})
    Alpha=@n1
    Translation="0,0,0"
    Rotation="-5,0,0"
    Scale="1,1,1"
   ) {
    rotationMode=BMM_Additive
    rotationSpace=BCS_BoneSpace
    alphaBoolBlend {blendOption=Linear}
    boneToModify {boneName=upperarm_l}
    rotation {pitch=-5}
   }
   n3 = "Transform (Modify) Bone - Bone: upperarm_r"(ComponentPose=@n2, Alpha=@n1, Translation="0,0,0", Rotation="1.8,0,0", Scale="1,1,1") {
    rotationMode=BMM_Additive
    rotationSpace=BCS_BoneSpace
    alphaBoolBlend {blendOption=Linear}
    boneToModify {boneName=upperarm_r}
    rotation {pitch=1.8}
   }
   Output Pose(Result=Component To Local(ComponentPose=@n3)) {layerGroup=None}
   n4 = Get bIsInFrontEnd()
   "To Float (Boolean)"(InBool=@n4)
   "To Float (Boolean)"(InBool=@n4)
  ]}
 EventGraph {nodes [
   Event BlueprintUpdateAnimation()
   > n1 = Branch(Condition=Montage_IsActive!(
     self:Anim Instance Object Reference=Get AnimInst() -> (AnimInst:Anim Instance Object Reference)
     Montage:Anim Montage Object Reference=/Game/Animation/Game/MainPlayer/Menu/BR/Female_Commando_Idle_03_Montage
    ) -> (ReturnValue:bool))
   > Set FrontEndAlpha(FrontEndAlpha:double=1) -> (Output_Get:double)
   Set FrontEndAlpha(execute=@n1.else, FrontEndAlpha=0)
   Event BlueprintInitializeAnimation()
   n2 = GetOwnerSkeletalMeshComponent()
   > Is Valid(InputObject=@n2)
   > Set SkelMesh(SkelMesh=@n2)
   > n3 = Cast To FortPlayerPawn(Object=GetOwningActor())
   > Is Valid(InputObject=@n3)
   > n4 = Set Pawn(Pawn:Fort Player Pawn Object Reference=@n3) -> (Output_Get:Fort Player Pawn Object Reference)
   n5 = GetAnimInstance(self=Get Mesh(self=@n4))
   > Is Valid(InputObject=@n5)
   > Set AnimInst(AnimInst:Anim Instance Object Reference=@n5) -> (Output_Get:Anim Instance Object Reference)
   TryGetPawnOwner()
  ]}
]
variables [
 AnimInst {type=AnimInstance, default=null}
 FrontEndAlpha {type=number}
 Pawn {type=FortPlayerPawn}
 SkelMesh {type=SkeletalMeshComponent}
]
